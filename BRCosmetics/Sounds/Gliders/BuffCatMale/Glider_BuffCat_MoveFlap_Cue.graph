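# Sound cue for the flap sound of the BuffCat glider as it moves.
# A Random node picks one of five equally weighted flap wave variations, played at a low volume multiplier.
# It uses the player glider thrust attenuation and a glider landing sound class.
# A concurrency override, used in place of the assigned open/close concurrency set, limits simultaneous plays per owner and stops the oldest.
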SoundCue /BRCosmetics/Sounds/Gliders/BuffCatMale/Glider_BuffCat_MoveFlap_Cue
root=Random_0
nodes [
 {id=Random_0, children [WavePlayer_1,WavePlayer_3,WavePlayer_4,WavePlayer_5,WavePlayer_6], properties {weights [1,1,1,1,1]}}
 {id=WavePlayer_1, properties {soundWaveAssetPtr=&/Glider_BuffCat_MoveFlap_01}}
 {id=WavePlayer_3, properties {soundWaveAssetPtr=&/Glider_BuffCat_MoveFlap_03}}
 {id=WavePlayer_4, properties {soundWaveAssetPtr=&/Glider_BuffCat_MoveFlap_04}}
 {id=WavePlayer_5, properties {soundWaveAssetPtr=&/Glider_BuffCat_MoveFlap_05}}
 {id=WavePlayer_6, properties {soundWaveAssetPtr=&/Glider_BuffCat_MoveFlap_06}}
]
properties {
 attenuationSettings=&Fort_Player_Glider_Thrust_Athena_Att
 bOverrideConcurrency=true
 maxDistance=12500
 soundClassObject=&Fort_Gliders_Land
 volumeMultiplier=0.07
 concurrencyOverrides {bLimitToOwner=true, maxCount=2, resolutionRule=StopOldest}
 concurrencySet [&Athena_GliderOpenClose_Conc]
}
