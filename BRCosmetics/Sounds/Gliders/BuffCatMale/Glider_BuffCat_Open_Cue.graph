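# Sound cue for the opening sound of the BuffCat glider cosmetic.
# A single wave player plays the Glider_BuffCat_Open_02 wave through a short fixed delay.
# A glider open property sheet, a shared open/close wide-pan attenuation and an open/close concurrency set govern playback, with the sound routed to the Fort_Gliders_Land class.
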SoundCue /BRCosmetics/Sounds/Gliders/BuffCatMale/Glider_BuffCat_Open_Cue
root=Delay_0
nodes [
 {id=Delay_0, children [WavePlayer_0], properties {delayMax=0.07, delayMin=0.07}}
 {id=WavePlayer_0, properties {soundWaveAssetPtr=&/Glider_BuffCat_Open_02}}
]
properties {
 attenuationSettings=/BRCosmetics/Sounds/Other/Attenuation/Fort_Player_Glider_OpenClose_Athena_WidePan_Att
 audioPropertiesSheet=/BRCosmetics/Sounds/Settings/PropertySheets/Glider/APS_Cosmetic_Glider_Open
 maxDistance=10000
 soundClassObject=&Fort_Gliders_Land
 volumeMultiplier=1.05
 concurrencySet [&Athena_GliderOpenClose_Conc]
}
