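# Sound cue for the thrust sound of the BuffCat male cosmetic glider.
# A Random node picks one of four thrust wave variations with equal weight.
# The cue plays quietly at slightly raised pitch, using shared glider thrust attenuation and concurrency settings.
# It is assigned to the glider land sound class and a cosmetic glider loop audio property sheet, though no wave player here sets looping.
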SoundCue /BRCosmetics/Sounds/Gliders/BuffCatMale/Glider_BuffCat_Thrust_Cue
root=Random_1
nodes [
 {id=Random_1, children [WavePlayer_0,WavePlayer_2,WavePlayer_4,WavePlayer_5], properties {weights [1,1,1,1]}}
 {id=WavePlayer_0, properties {soundWaveAssetPtr=&/Glider_BuffCat_Thrust_01}}
 {id=WavePlayer_2, properties {soundWaveAssetPtr=&/Glider_BuffCat_Thrust_02}}
 {id=WavePlayer_4, properties {soundWaveAssetPtr=&/Glider_BuffCat_Thrust_03}}
 {id=WavePlayer_5, properties {soundWaveAssetPtr=&/Glider_BuffCat_Thrust_04}}
]
properties {
 attenuationSettings=&Fort_Player_Glider_Thrust_Athena_Att
 audioPropertiesSheet=&APS_Cosmetic_Glider_Loop
 maxDistance=12500
 pitchMultiplier=1.1
 soundClassObject=&Fort_Gliders_Land
 volumeMultiplier=0.1
 concurrencySet [&Athena_Glider_Thrust_Conc]
}
